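# First-person looping healing sound cue for the FizzyCloud heal spray, routed to the consumables sound class.
# A mixer layers a louder start one-shot, with slightly randomized volume, over a quieter looping heal wave.
# An enveloper fades the loop in briefly and ramps its pitch steadily upward over several seconds, so the sound rises as playback continues.
# A concurrency override caps simultaneous instances.
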SoundCue /FizzyCloudGameplay/GameplayCues/Assets/Audio/FizzyCloud_HealSpray_Heal_1P_Loop_Healing_Cue
root=Mixer_1
nodes [
 {
  id=Enveloper_1
  children [WavePlayer_12]
  properties {pitchCurve {editorCurveData {keys [<at 0 0.65>,<at 12 2>]}}, volumeCurve {editorCurveData {keys [<at 0 0>,<at 1 1>]}}}
 }
 {id=Mixer_1, children [Modulator_0,Enveloper_1], properties {inputVolume [1.5,0.5]}}
 {id=Modulator_0, children [WavePlayer_13], properties {volumeMax=0.8, volumeMin=0.6}}
 {id=WavePlayer_12, properties {bLooping=true, soundWaveAssetPtr=&/FizzyCloud_Player_Heal_Loop}}
 {id=WavePlayer_13, properties {soundWaveAssetPtr=&/FizzyCloud_Player_Heal_Loop_Start_01}}
]
properties {
 bOverrideConcurrency=true
 maxDistance=2097152
 soundClassObject=&Fort_Consumables
 volumeMultiplier=1
 concurrencyOverrides {maxCount=3}
}
