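# Sound cue for the first-person loop-stop sound of the FizzyCloud heal spray's healing effect.
# A single wave player feeds a Modulator that randomizes pitch with a raised upper bound while holding volume constant.
# It routes to the consumables sound class at slightly reduced volume, with overridden concurrency capping it at a few simultaneous instances.
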SoundCue /FizzyCloudGameplay/GameplayCues/Assets/Audio/FizzyCloud_HealSpray_Heal_1P_Loop_Stop_Cue
root=Modulator_0
nodes [
 {id=Modulator_0, children [WavePlayer_3], properties {pitchMax=1.2, volumeMax=1, volumeMin=1}}
 {id=WavePlayer_3, properties {soundWaveAssetPtr=&/FizzyCloud_Player_Heal_Loop_Stop_01}}
]
properties {
 bOverrideConcurrency=true
 maxDistance=2097152
 soundClassObject=&Fort_Consumables
 volumeMultiplier=0.8
 concurrencyOverrides {maxCount=3}
}
