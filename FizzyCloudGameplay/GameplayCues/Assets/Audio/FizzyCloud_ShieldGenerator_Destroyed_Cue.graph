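# SoundCue for the destruction sound of the FizzyCloud shield generator gadget.
# A Random node picks one of two equally weighted ShieldSprinkler destroy waves and feeds it through a Modulator.
# It plays in the gadget items sound class with boosted volume, using overridden attenuation alongside a referenced grenade-bounce attenuation asset.
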SoundCue /FizzyCloudGameplay/GameplayCues/Assets/Audio/FizzyCloud_ShieldGenerator_Destroyed_Cue
root=Modulator_0
nodes [
 {id=Modulator_0, children [Random_0]}
 {id=Random_0, children [WavePlayer_4,WavePlayer_5], properties {weights [1,1]}}
 {id=WavePlayer_4, properties {soundWaveAssetPtr=&/FizzyCloud_ShieldSprinkler_Destroy_01}}
 {id=WavePlayer_5, properties {soundWaveAssetPtr=&/FizzyCloud_ShieldSprinkler_Destroy_02}}
]
properties {
 attenuationSettings=/Game/Sounds/Attenuation/Fort_Grenade_Bounce_Att
 bOverrideAttenuation=true
 maxDistance=2450
 soundClassObject=&Fort_Gadgets_Items
 volumeMultiplier=1.3
 attenuationOverrides {falloffDistance=2400, stereoSpread=2000, attenuationShapeExtents {x=50}}
}
