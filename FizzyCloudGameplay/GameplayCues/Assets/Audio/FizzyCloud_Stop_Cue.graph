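# Sound cue for the Fizzy Cloud stop sound, stored with the FizzyCloudGameplay gameplay cue audio assets.
# An equal-weight Random node picks one of three stop waves, feeding a volume Modulator and then a root Attenuation node.
# Both that node and the cue override the player swimming attenuation, with the cue adding occlusion and logarithmic falloff.
# The cue uses the water interaction sound class, with raised volume and pitch.
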SoundCue /FizzyCloudGameplay/GameplayCues/Assets/Audio/FizzyCloud_Stop_Cue
root=Attenuation_0
nodes [
 {
  id=Attenuation_0
  children [Modulator_0]
  properties {
   attenuationSettings=/Game/Sounds/Attenuation/Fort_Player_Swimming_Athena_Att
   bOverrideAttenuation=true
   attenuationOverrides {falloffDistance=500, attenuationShapeExtents {x=0}}
  }
 }
 {id=Modulator_0, children [Random_1], properties {volumeMax=1.25, volumeMin=1.2}}
 {id=Random_1, children [WavePlayer_3,WavePlayer_4,WavePlayer_5], properties {bShouldExcludeFromBranchCulling=true, weights [1,1,1]}}
 {id=WavePlayer_3, properties {soundWaveAssetPtr=&/FizzyCloud_Stop1}}
 {id=WavePlayer_4, properties {soundWaveAssetPtr=&/FizzyCloud_Stop2}}
 {id=WavePlayer_5, properties {soundWaveAssetPtr=&/FizzyCloud_Stop3}}
]
properties {
 attenuationSettings=&Fort_Player_Swimming_Athena_Att
 bOverrideAttenuation=true
 maxDistance=2200
 pitchMultiplier=1.1
 soundClassObject=/Game/Sounds/FortSoundClassesAndModes/Fort_Water_Interaction
 volumeMultiplier=2
 attenuationOverrides {
  bEnableOcclusion=true
  distanceAlgorithm=Logarithmic
  falloffDistance=1800
  nonSpatializedRadiusStart=600
  occlusionLowPassFilterFrequency=18000
  occlusionTraceChannel=ECC_GameTraceChannel11
  occlusionVolumeAttenuation=0.6
  stereoSpread=2000
 }
}
