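# Sound cue for the Fizzy Cloud detonation, stored with that gameplay feature's GameplayCue audio.
# A single MedMist_Detonation wave player feeds a Modulator at default settings as the root.
# It plays in the Fort_Gadgets sound class at slightly reduced volume, using the shared smoke grenade attenuation asset.
# A concurrency override caps how many instances play at once.
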SoundCue /FizzyCloudGameplay/GameplayCues/Assets/Audio/FizzyCloud_detonate_Cue
root=Modulator_0
nodes [
 {id=Modulator_0, children [WavePlayer_12]}
 {id=WavePlayer_12, properties {soundWaveAssetPtr=&/MedMist_Detonation}}
]
properties {
 attenuationSettings=/Game/Sounds/Attenuation/Fort_Grenade_Smoke_Att
 bOverrideConcurrency=true
 maxDistance=6800
 soundClassObject=&Fort_Gadgets
 volumeMultiplier=0.8
 attenuationOverrides {
  bAttenuateWithLPF=true
  falloffDistance=10000
  lPFRadiusMax=10000
  occlusionLowPassFilterFrequency=18000
  occlusionTraceChannel=ECC_GameTraceChannel11
  occlusionVolumeAttenuation=0.8
  stereoSpread=2000
  attenuationShapeExtents {x=0}
 }
 concurrencyOverrides {maxCount=3}
}
