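# Sound cue for the Fizzy Cloud in/out effect, stored with the gameplay cue audio assets.
# A root Random node picks one of three equally weighted InOut wave players on each play and is excluded from branch culling.
# The cue doubles volume and carries a very large max distance.
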SoundCue /FizzyCloudGameplay/GameplayCues/Assets/Audio/SC_FizzyCloud_InOut_Cue
root=Random_0
nodes [
 {id=Random_0, children [WavePlayer_0,WavePlayer_1,WavePlayer_2], properties {bShouldExcludeFromBranchCulling=true, weights [1,1,1]}}
 {id=WavePlayer_0, properties {soundWaveAssetPtr=&/FizzyCloud_InOut_1}}
 {id=WavePlayer_1, properties {soundWaveAssetPtr=&/FizzyCloud_InOut_2}}
 {id=WavePlayer_2, properties {soundWaveAssetPtr=&/FizzyCloud_InOut_3}}
]
properties {maxDistance=2097152, volumeMultiplier=2}
